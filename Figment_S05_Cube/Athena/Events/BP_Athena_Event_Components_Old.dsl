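; Season 5 cube event actor that drives the sky crack, its scheduled lightning strikes and the final beam that closes the rift.
; Calendar events set crack opacity and corruption, applied through OnRep functions, and time the recurring strikes and final beam.
; Strike strength and cactus target derive from remaining event time.
; After a warmup, authority destroys cacti and player builds along the bolt, while clients play particles, audio and rift closure.
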(event Custom|DoLighting
  ;[1 1 1] Lightning Logic
  (if (Variables|Default|GetCanFireLightning)
    (bind (_biseventactive _timesincebegin _timeuntilend) (CalendarEvents|GetCalendarEventInformation "EventFlag.BR_S5_Cube_Lightning"))
    (bind (_strength _hitindex _timeuntilnext) (CallFunction|SelectStrengthandHitLocation _timeuntilend))
    (Variables|Default|SetTime%20Until%20Next _timeuntilnext)
    (CallFunction|SingleLightning _strength (Math|Conversions|ToInteger(Byte) _hitindex))
    (bind _returnvalue (Utilities|Time|SetTimerbyEvent (EventDispatchers|CreateEvent :!event Custom|DoLighting) (Variables|Default|GetTime%20Until%20Next)))
    (Variables|Default|SetDoLightningTImer _returnvalue)))

(event Custom|ServerWarmupDelay
  ;[1 1 1] Single Lightining Strike
  (Default|CallDestroyCactus (stale Custom|SingleLightning HitLocationIndex))
  (Class|CubeAppearEffectsFigment|LightingStrikeEffects (Variables|Default|GetLightningEffectsActor) (Variables|Default|GetLightningStrikePoint) (Utilities|Select :Index (stale Custom|SingleLightning BoltStrength) :Option%200 0.0 :Option%201 0.0 :Option%202 0.0 :Option%203 3.0 :Option%204 (- (* (Variables|Default|GetMinutesOfFinalLightning) 6e+01) 2.5)))
  ;; TraceforPBW
  (bind _outhits (Collision|MultiSphereTraceForObjects (Variables|Default|GetCenterCrackLocation) (Variables|Default|GetLightningStrikePoint) 256.0 (Utilities|Array|MakeArray "ObjectTypeQuery1")))
  (for _array_element _outhits
    (bind _asbuilding_static_mesh (Utilities|Casting|CastToBuildingSMActor (.HitActor (Collision|BreakHitResult _array_element))))
    (if (BuildingActor|IsPlayerBuilt _asbuilding_static_mesh)
      (Actor|DestroyActor (.HitActor (Collision|BreakHitResult _array_element))))))

(event Custom|SingleLightning
  (params (BoltStrength :type Byte) (HitLocationIndex :type Integer))
  (bind _lightningstrikepoint (Variables|Default|GetLightningStrikePoint))
  ;[1 1 1] Single Lightining Strike
  (Utilities|Time|ClearandInvalidateTimerbyHandle (Variables|Default|GetServerWarmupDelayTimer))
  (Utilities|Time|ClearandInvalidateTimerbyHandle (Variables|Default|GetClientWarmupDelayTimer))
  (Variables|Default|SetLightningStrikePoint (Utilities|Array|Get(acopy) (Variables|Default|GetLightningStrikePoints) HitLocationIndex))
  (Networking|SwitchHasAuthority
    (:Authority
      (bind _returnvalue (Utilities|Time|SetTimerbyEvent (delegate Custom|ServerWarmupDelay) (+ (Variables|Default|GetLightningWarmupTime) 0.1)))
      (Variables|Default|SetServerWarmupDelayTimer _returnvalue))
    (:Remote
      (Variables|Default|SetEffectRotation (Math|Rotator|FindLookatRotation (Variables|Default|GetLightningStrikePoint) (Variables|Default|GetCenterCrackLocation)))
      (Effects|SpawnEmitteratLocation "/Figment_S05/Athena/Prototype/Blueprints/Cube/FX/P_Lightning_Strike_Warmup_Red.P_Lightning_Strike_Warmup_Red" _lightningstrikepoint (Variables|Default|GetEffectRotation))
      (CallFunction|BeginLightningStrikeAudio _lightningstrikepoint BoltStrength)
      (bind _returnvalue_2 (Utilities|Time|SetTimerbyEvent (delegate Custom|ClientWarmupDelay) (Variables|Default|GetLightningWarmupTime)))
      (Variables|Default|SetClientWarmupDelayTimer _returnvalue_2))))

(event Custom|ClientWarmupDelay
  (bind _minutesoffinallightning (Variables|Default|GetMinutesOfFinalLightning))
  (bind _returnvalue (* _minutesoffinallightning 6e+01))
  (bind _returnvalue_2 (- _returnvalue 2.5))
  (bind _centralcrack (Variables|Default|GetCentralCrack))
  (bind (_returnvalue_location _returnvalue_rotation) (Math|Transform|BreakTransform (Transformation|GetWorldTransform _centralcrack)))
  (bind _crackaudioloop (Variables|Default|GetCrackAudioLoop))
  (bind _returnvalue_3 (Transformation|GetWorldLocation _crackaudioloop))
  ;[1 1 1] Single Lightining Strike
  (bind _returnvalue_4 (Effects|SpawnEmitteratLocation (Utilities|Select :Index (stale Custom|SingleLightning BoltStrength) :Option%200 "/Figment_S05/Athena/Prototype/Blueprints/Cube/FX/P_Lightning_Strike_Small_Red.P_Lightning_Strike_Small_Red" :Option%201 "/Figment_S05/Athena/Prototype/Blueprints/Cube/FX/P_Lightning_Strike_Med_Red.P_Lightning_Strike_Med_Red" :Option%202 "/Figment_S05/Athena/Prototype/Blueprints/Cube/FX/P_Lightning_Strike_Large_Red.P_Lightning_Strike_Large_Red" :Option%203 "/Figment_S05/Athena/Prototype/Blueprints/Cube/FX/P_Lightning_Strike_Beam_Red.P_Lightning_Strike_Beam_Red" :Option%204 "/Figment_S05/Athena/Prototype/Blueprints/Cube/FX/P_Lightning_Strike_Beam_Final_loop_Red.P_Lightning_Strike_Beam_Final_loop_Red") (Variables|Default|GetLightningStrikePoint) (Variables|Default|GetEffectRotation)))
  (bind _output_get (Variables|Default|SetCurrentLightningEmitter _returnvalue_4))
  (EventDispatchers|BindEventtoOnParticleSpawn _output_get (delegate Custom|On%20Beam%20Pulse))
  (GameplayCue|ExecuteGameplayCue%28Burst%29Local self (Variables|Default|GetGameplay%20Cue%20Tag) (Ability|GameplayCue|MakeGameplayCueParameters :NormalizedMagnitude 0.0 :RawMagnitude 0.0 :EffectContext "()" :AggregatedSourceTags "(GameplayTags=)" :AggregatedTargetTags "(GameplayTags=)" :Location (Math|Vector|MakeVectorNetQuantize10 (.x (Variables|Default|GetLightningStrikePoint)) (.y (Variables|Default|GetLightningStrikePoint)) (.z (Variables|Default|GetLightningStrikePoint))) :Normal (Math|Vector|MakeVectorNetQuantizeNormal 0.0 0.0 1.0) :GameplayEffectLevel 1 :AbilityLevel 1 :bReplicateLocationWhenUsingMinimalRepProxy false))
  (Utilities|FlowControl|Delay 0.1) ; Give bolt time to reach the ground
  (Class|CubeAppearEffectsFigment|LightingStrikeEffects (Variables|Default|GetLightningEffectsActor) (Variables|Default|GetLightningStrikePoint) (Utilities|Select :Index (stale Custom|SingleLightning BoltStrength) :Option%200 0.0 :Option%201 0.0 :Option%202 0.0 :Option%203 3.0 :Option%204 _returnvalue_2))
  (if (Math|Byte|Equal(Byte) (stale Custom|SingleLightning BoltStrength) 4)
    (Utilities|FlowControl|Delay (Utilities|Select :Index (stale Custom|SingleLightning BoltStrength) :Option%200 0.0 :Option%201 0.0 :Option%202 0.0 :Option%203 3.0 :Option%204 _returnvalue_2)) ; Wait for final beam to finish
    ;[1 1 1] Close up sky rift
    (Rendering|SetVisibility _centralcrack false true)
    (bind _returnvalue_5 (Effects|SpawnEmitteratLocation "/Figment_S05/Athena/Prototype/Blueprints/Cube/FX/P_Geode_Rift_Close_Red.P_Geode_Rift_Close_Red" _returnvalue_location _returnvalue_rotation)) ; Crack close FX
    (EventDispatchers|BindEventtoOnParticleSpawn _returnvalue_5 (delegate Custom|On%20Final%20Flare%20Spawn))
    (Audio|SpawnSoundatLocation (Variables|Default|GetOnRiftCloseSound) _returnvalue_3)
    (Fortnite|PlayLocalForceFeedbackatLocation "/Figment_S05/Athena/Maps/Test/Events/FF_Geode_Closing.FF_Geode_Closing" _returnvalue_3 5e+05)
    (Components|Activation|SetActive (Variables|Default|GetP_Geode_Crack_Center_Loop) false false)
    (CallFunction|FadeLightningLoopAudio 0)
    (Audio|Components|Audio|FadeOut (Variables|Default|GetCrackAudioLoop) 5.0 0.0 "Linear" :!class /Script/Engine.AudioComponent)
    (elif (Math|Byte|Equal(Byte) (stale Custom|SingleLightning BoltStrength) 3)
      (Utilities|FlowControl|Delay (Utilities|Select :Index (stale Custom|SingleLightning BoltStrength) :Option%200 0.0 :Option%201 0.0 :Option%202 0.0 :Option%203 3.0 :Option%204 _returnvalue_2)) ; Fade audio on short beam
      (CallFunction|FadeLightningLoopAudio 0))))

(event Custom|On%20Beam%20Pulse
  (params (EventName :type Name) (EmitterTime :type Float%20%28single-precision%29) (Location :type Vector) (Velocity :type Vector))
  ;[1 1 1] Single Lightining Strike
  (if (Utilities|Name|Equal(Name) EventName "crack")
    (Audio|SpawnSoundatLocation "/FNGameplayCues/Athena/Sounds/Events/CrackLightning/Crack_Lightning_Pulse_Cue.Crack_Lightning_Pulse_Cue" (Transformation|GetWorldLocation (Variables|Default|GetCrackAudioLoop)))))

(event Custom|On%20Final%20Flare%20Spawn
  (params (EventName :type Name) (EmitterTime :type Float%20%28single-precision%29) (Location :type Vector) (Velocity :type Vector))
  (bind _crackaudioloop (Variables|Default|GetCrackAudioLoop))
  (bind _returnvalue (Transformation|GetWorldLocation _crackaudioloop))
  ;[1 1 1] Close up sky rift
  (if (Utilities|Name|Equal(Name) EventName "flare")
    (Audio|SpawnSoundatLocation (Variables|Default|GetOnRiftCloseFinalSound) _returnvalue)
    (Fortnite|PlayLocalForceFeedbackatLocation "/Figment_S05/Athena/Maps/Test/Events/Rocket_BigBang.Rocket_BigBang" _returnvalue 5e+05)))

(event Custom|Begin%20Lightning%20Strike%20Audio
  (params (Strike%20Location :type Vector) (Index :type Byte))
  (bind _crackaudioloop (Variables|Default|GetCrackAudioLoop))
  (bind _returnvalue (Transformation|GetWorldLocation _crackaudioloop))
  ;[1 1 1] Lightning audio
  (Audio|PlaySoundatLocation "/Figment_S05/Athena/Sounds/Events/CrackLightning/Crack_Lightning_Warmup_Cue.Crack_Lightning_Warmup_Cue" (Transformation|GetWorldLocation (Variables|Default|GetCrackAudioLoop)))
  (Utilities|FlowControl|Sequence
    (:then_0
      (Utilities|FlowControl|Delay (- (Variables|Default|GetLightningWarmupTime) (Variables|Default|GetLightningStrikeAudioOffset)))
      (Audio|SpawnSoundatLocation (Utilities|Select :Index Index :Option%200 "/Figment_S05/Athena/Sounds/Events/CrackLightning/Crack_Lightning_Strike_Cue.Crack_Lightning_Strike_Cue" :Option%201 "/Figment_S05/Athena/Sounds/Events/CrackLightning/Crack_Lightning_Strike_Cue.Crack_Lightning_Strike_Cue" :Option%202 "/Figment_S05/Athena/Sounds/Events/CrackLightning/Crack_Lightning_Strike_Cue.Crack_Lightning_Strike_Cue" :Option%203 "/Figment_S05/Athena/Sounds/Events/CrackLightning/Crack_Lightning_Strike_Final_Cue.Crack_Lightning_Strike_Final_Cue" :Option%204 "/Figment_S05/Athena/Sounds/Events/CrackLightning/Crack_Lightning_Strike_Final_Cue.Crack_Lightning_Strike_Final_Cue") _returnvalue)
      (bind _returnvalue_2 (Audio|SpawnSoundatLocation (Utilities|Select :Index Index :Option%203 "/Figment_S05/Athena/Sounds/Events/CrackLightning/Crack_Lightning_Strike_Final_Loop_Cue.Crack_Lightning_Strike_Final_Loop_Cue" :Option%204 "/Figment_S05/Athena/Sounds/Events/CrackLightning/Crack_Lightning_Strike_Final_Loop_Cue.Crack_Lightning_Strike_Final_Loop_Cue") _returnvalue))
      (bind _output_get (Variables|Default|SetStrikeAudio _returnvalue_2))
      (Audio|Components|Audio|FadeIn _output_get 0.5 1.0 0.0 "Linear" :!class /Script/Engine.AudioComponent))
    (:then_1
      (Utilities|FlowControl|Delay (Variables|Default|GetLightningWarmupTime))
      (Audio|SpawnSoundatLocation (Utilities|Select :Index Index :Option%200 "/Figment_S05/Athena/Sounds/Events/CrackLightning/Crack_Lightning_Strike_Impact_Cue.Crack_Lightning_Strike_Impact_Cue" :Option%201 "/Figment_S05/Athena/Sounds/Events/CrackLightning/Crack_Lightning_Strike_Impact_Cue.Crack_Lightning_Strike_Impact_Cue" :Option%202 "/Figment_S05/Athena/Sounds/Events/CrackLightning/Crack_Lightning_Strike_Impact_Cue.Crack_Lightning_Strike_Impact_Cue" :Option%203 "/Figment_S05/Athena/Sounds/Events/CrackLightning/Crack_Lightning_Strike_Impact_Final_Cue.Crack_Lightning_Strike_Impact_Final_Cue" :Option%204 "/Figment_S05/Athena/Sounds/Events/CrackLightning/Crack_Lightning_Strike_Impact_Final_Cue.Crack_Lightning_Strike_Impact_Final_Cue") Strike%20Location)
      (bind _returnvalue_3 (Audio|SpawnSoundatLocation (Utilities|Select :Index Index :Option%203 "/Figment_S05/Athena/Sounds/Events/CrackLightning/Crack_Lightning_Strike_Impact_Final_Loop_Cue.Crack_Lightning_Strike_Impact_Final_Loop_Cue" :Option%204 "/Figment_S05/Athena/Sounds/Events/CrackLightning/Crack_Lightning_Strike_Impact_Final_Loop_Cue.Crack_Lightning_Strike_Impact_Final_Loop_Cue") Strike%20Location))
      (bind _output_get_2 (Variables|Default|SetStrikeImpactAudio _returnvalue_3))
      (Audio|Components|Audio|FadeIn _output_get_2 0.5 1.0 0.0 "Linear" :!class /Script/Engine.AudioComponent))))

(event Custom|TestCracks
  (params (DaysRemaining :type Float%20%28double-precision%29))
  ;[1 1 1] Test crack progression by days remaining
  (Variables|Default|SetCrackOpacity (Math|Float|MapRangeClamped DaysRemaining 0.0 (Variables|Default|GetCrackEventDuration) 0.0 1.0)))

(event Custom|TestCracksAnimated
  ;[1 1 1] Test crack spread animation
  (bind (_direction _newtrack_0 _corruption) (Timeline|CrackSpreadTest PlayFromStart)
    (:Update
      (Variables|Default|SetCrackOpacity _newtrack_0)
      (Variables|Default|SetCorruption _corruption))))

(event EventBeginPlay
  ;[0 0.07145 1] Begin Play
  (Networking|CanExecuteCosmeticEvents
    (:False
      (bind _asfort_game_mode_zone (Utilities|Casting|CastToFortGameModeZone (Game|GetGameMode)))
      (EventDispatchers|BindEventtoOnHandleMatchHasStarted _asfort_game_mode_zone (EventDispatchers|CreateEvent :!event Custom|ContinueSetup)))))

(event Custom|Fade%20Lightning%20Loop%20Audio
  (params (Index :type Byte))
  ;[1 1 1] Lightning audio
  (Utilities|FlowControl|Sequence
    (:then_0
      (Audio|PlaySoundatLocation (Utilities|Select :Index Index :Option%200 "/Figment_S05/Athena/Sounds/Events/CrackLightning/Crack_Lightning_Strike_Final_Cue.Crack_Lightning_Strike_Final_Cue" :Option%201 "/Figment_S05/Athena/Sounds/Events/CrackLightning/Crack_Lightning_Strike_Final_Cue.Crack_Lightning_Strike_Final_Cue") (Transformation|GetWorldLocation (Variables|Default|GetStrikeAudio)) "0, 0, 0" 1.0 1.5)
      (Audio|PlaySoundatLocation (Utilities|Select :Index Index :Option%200 "/Figment_S05/Athena/Sounds/Events/CrackLightning/Crack_Lightning_Strike_Impact_Final_Cue.Crack_Lightning_Strike_Impact_Final_Cue") (Transformation|GetWorldLocation (Variables|Default|GetStrikeImpactAudio))))
    (:then_1
      (bind _strikeimpactaudio (Variables|Default|GetStrikeImpactAudio)
        (:then
          (Audio|Components|Audio|FadeOut _strikeimpactaudio 1.0 0.0 "Linear" :!class /Script/Engine.AudioComponent))
        (:else)))
    (:then_2
      (bind _strikeaudio (Variables|Default|GetStrikeAudio)
        (:then
          (Audio|Components|Audio|FadeOut _strikeaudio 1.0 0.0 "Linear" :!class /Script/Engine.AudioComponent))
        (:else)))))

(event Custom|ContinueSetup
  (Utilities|FlowControl|Delay 1.0)
  (Utilities|FlowControl|Sequence
    (:then_0
      ;[1 1 1] Crack Effects
      (bind (_eventname _timeuntilend) (MCPContext|WatchCalendarEvent (Variables|Default|GetCrackShrinkEvent))
        (:OnEventActive
          (Variables|Default|SetCrackOpacity (Math|Float|MapRangeClamped (Math|Timespan|GetTotalDays _timeuntilend) 0.0 (Variables|Default|GetCrackEventDuration) 0.0 1.0)))
        (:OnEventEnded
          (Variables|Default|SetCrackOpacity 0.0))))
    (:then_1
      (bind (_eventname_2 _timeuntilend_2) (MCPContext|WatchCalendarEvent (Variables|Default|GetCrackCorruptEvent))
        (:OnEventActive
          (Variables|Default|SetCrackOpacity 0.0)
          (Variables|Default|SetCorruption (Math|Float|MapRangeClamped (Math|Timespan|GetTotalDays _timeuntilend_2) 0.0 (Variables|Default|GetCrackCorruptionEventDuration) 1.0)))))
    (:then_2
      (bind (_eventname_3 _timeuntilend_3 _timesincebegin) (MCPContext|WatchCalendarEvent "EventFlag.BR_S5_Cube_Lightning")
        (:OnEventActive
          _)
        (:OnEventUpdated
          _))
      (Variables|Default|SetTotalTimeEvent (Math|Timespan|Timespan+Timespan _timeuntilend_3 _timesincebegin))
      (Utilities|FlowControl|Sequence
        (:then_0
          ;[1 1 1] Activate Final Lightning- an exact time from end of event
          (if (<= (Math|Timespan|GetTotalMinutes _timeuntilend_3) (+ (Variables|Default|GetMinutesOfFinalLightning) 6e+01))
            (Variables|Default|SetFinalLightningSeconds (- (* (- (Math|Timespan|GetTotalMinutes _timeuntilend_3) (Variables|Default|GetMinutesOfFinalLightning)) 6e+01) (Variables|Default|GetLightningWarmupTime)))
            (Utilities|Time|SetTimerbyEvent (EventDispatchers|CreateEvent :!event Custom|EnableFinalLightning) (Math|Float|Max(Float) (Variables|Default|GetFinalLightningSeconds) 0.05))
            (Utilities|Time|SetTimerbyEvent (EventDispatchers|CreateEvent :!event Custom|DisableFinalLightning) (Math|Float|float+float (Variables|Default|GetFinalLightningSeconds) (* (Variables|Default|GetMinutesOfFinalLightning) 6e+01) (Variables|Default|GetBeamEndOffsetSeconds)))))
        (:then_1
          ;[1 1 1] Set Current State of Cactus Countdown
          (bind (_strength _hitindex) (CallFunction|SelectStrengthandHitLocation _timeuntilend_3))
          (if (== (Math|Conversions|ToInteger(Byte) _hitindex) 1)
            (Utilities|FlowControl|Delay 5.0))
          (Default|CallCactusCalculationsComplete (Math|Conversions|ToInteger(Byte) _hitindex)))
        (:then_2
          ;[1 1 1] Do first lightning strike if event JUST fired!
          (if (< (Math|Timespan|GetTotalSeconds _timesincebegin) 0.5)
            (CallFunction|DoLighting)
            (else
              ;[1 1 1] Calculate Next Lighting Strike
              (bind _timeuntilnext (CallFunction|CalculateTimetoNextStrike _timeuntilend_3))
              (bind _returnvalue (Utilities|Time|SetTimerbyEvent (EventDispatchers|CreateEvent :!event Custom|DoLighting) _timeuntilnext))
              (Variables|Default|SetDoLightningTImer _returnvalue))))))))

(event Custom|Final
  (Variables|Default|SetFinalLightningSeconds (- (* (- (/ 25.0 6e+01) (Variables|Default|GetMinutesOfFinalLightning)) 6e+01) (Variables|Default|GetLightningWarmupTime)))
  ;[1 1 1] Activate Final Lightning- an exact time from end of event
  (Utilities|Time|SetTimerbyEvent (EventDispatchers|CreateEvent :!event Custom|EnableFinalLightning) (Math|Float|Max(Float) (Variables|Default|GetFinalLightningSeconds) 0.05))
  (Utilities|Time|SetTimerbyEvent (EventDispatchers|CreateEvent :!event Custom|DisableFinalLightning) (Math|Float|float+float (Variables|Default|GetFinalLightningSeconds) (* (Variables|Default|GetMinutesOfFinalLightning) 6e+01) (Variables|Default|GetBeamEndOffsetSeconds))))

(event Custom|EnableFinalLightning
  ;[1 1 1] Activate Final Lightning- an exact time from end of event
  (Variables|Default|SetFinalLightningActive true)
  (Utilities|Time|ClearandInvalidateTimerbyHandle (Variables|Default|GetDoLightningTImer))
  (Utilities|Time|ClearandInvalidateTimerbyHandle (Variables|Default|GetServerWarmupDelayTimer))
  (Variables|Default|SetCanFireLightning false)
  (Utilities|FlowControl|Delay 5.0)
  (Default|CallDestroyCactus 8)
  (Default|CallDestroyCactus 9))

(event Custom|DisableFinalLightning
  ;[1 1 1] Activate Final Lightning- an exact time from end of event
  (Variables|Default|SetFinalLightningActive false))

(fn CalculateTimeToNextStrike
  (params (TimeUntilEnd :type Timespan%20Structure))
  (for (_array_element _array_index) (Variables|Default|GetCountdownTiming)
    (if (> (- 1.0 (Math|Timespan|TimespanRatio TimeUntilEnd (Variables|Default|GetTotalTimeEvent))) _array_element)
      (LocalVariables|SetTestValue _array_index)
      (else
        (return (* (Math|Float|%25%28Float%29 (* (Math|Float|Fraction (Math|Timespan|GetTotalHours TimeUntilEnd)) 6e+01) (Utilities|Array|Get(acopy) (Variables|Default|GetCountdownFrequency) (LocalVariables|GetTestValue))) 6e+01) (LocalVariables|GetTestValue))))))

(fn OnRep_Corruption
  (bind _corruption (Variables|Default|GetCorruption))
  (Rendering|SetActorHiddenInGame false)
  (Rendering|Material|SetScalarParameterValueonMaterials (Variables|Default|GetCentralCrack) "Corruption" _corruption)
  (Audio|Parameter|SetFloatParameter (Variables|Default|GetCrackAudioLoop) "Corruption" _corruption :!class /Script/Engine.AudioComponent)
  (if (> _corruption 0.1)
    (Effects|Components|ParticleSystem|SetTemplate (Variables|Default|GetP_Geode_Crack_Center_Loop) "/Figment_S05/Athena/Effects/Particles/Geode/P_Geode_Crack_Center_ParticlesOnly_Corrupt_Red.P_Geode_Crack_Center_ParticlesOnly_Corrupt_Red")))

(fn OnRep_CrackProgression
  (bind _crackopacity (Variables|Default|GetCrackOpacity))
  (bind _centralcrack (Variables|Default|GetCentralCrack))
  (Rendering|SetActorHiddenInGame false)
  (Utilities|FlowControl|Sequence
    (:then_0
      (Rendering|SetVisibility (Variables|Default|GetSmallCrack1) (> _crackopacity 0.1) false)
      (Rendering|Material|SetScalarParameterValueonMaterials (Variables|Default|GetCentralCrack) "Opacity" (Variables|Default|GetCrackOpacity))
      (if (<= _crackopacity 0.0)
        (Components|StaticMesh|SetStaticMesh _centralcrack "/Figment_S05/Athena/Effects/Meshes/Geode/SM_Geode_Sky_Crack_Healed.SM_Geode_Sky_Crack_Healed")
        (Rendering|Material|SetScalarParameterValueonMaterials _centralcrack "MinOpacity" 0.5)))
    (:then_1
      ;[1 1 1] Start up our amb crack audio (defaulted to off now)
      (if (not (Audio|Components|Audio|IsPlaying (Variables|Default|GetCrackAudioLoop)))
        (Audio|Components|Audio|Play (Variables|Default|GetCrackAudioLoop) 0.0)))))

(fn OnRep_CracksVisible
  (if (Variables|Default|GetCracksVisible)
    (Rendering|SetVisibility (Variables|Default|GetCentralCrack) true true)
    (if (not (Audio|Components|Audio|IsPlaying (Variables|Default|GetCrackAudioLoop)))
      (Audio|Components|Audio|Play (Variables|Default|GetCrackAudioLoop) 0.0))))

(fn OnRep_FinalLightningActive
  (if (Variables|Default|GetFinalLightningActive)
    (CallFunction|SingleLightning 4)
    (else
      (Components|Activation|Deactivate (Variables|Default|GetCurrentLightningEmitter))
      (Utilities|Time|ClearandInvalidateTimerbyHandle (Variables|Default|GetServerWarmupDelayTimer))
      (Utilities|Time|ClearandInvalidateTimerbyHandle (Variables|Default|GetClientWarmupDelayTimer))
      (Utilities|Time|SetTimerbyFunctionName (Variables|Default|GetLightningEffectsActor) "DeactivateEffect" 1.5))))

(fn SelectStrengthAndHitLocation
  (params (TimeRemaining :type Timespan%20Structure))
  (LocalVariables|SetNormalizedTimeRemaining (- 1.0 (Math|Timespan|TimespanRatio TimeRemaining (Variables|Default|GetTotalTimeEvent))))
  (Utilities|FlowControl|Sequence
    (:then_0
      (Utilities|FlowControl|Sequence
        (:then_0
          (bind (_timeuntilnext _cactusindex) (CallFunction|CalculateTimetoNextStrike TimeRemaining))
          (LocalVariables|SetTimeForNext _timeuntilnext)
          (bind _output_get (Variables|Default|SetCactusHitIndex (Math|Conversions|ToByte(Integer) (+ _cactusindex 1))))
          (LocalVariables|SetAttemptedHitIndex _output_get))
        (:then_1
          ;[1 1 1] Calculate Strength
          (if (<= (Math|Timespan|GetTotalMinutes TimeRemaining) 3e+01)
            (LocalVariables|SetLightningStrength 3)
            (else
              (LocalVariables|SetLightningStrength (Math|Conversions|ToByte(Integer) (Math|Float|Floor (/ (LocalVariables|GetNormalizedTimeRemaining) (/ 1.0 4.0))))))))))
    (:then_1
      ;[1 1 1] Return
      (return (LocalVariables|GetLightningStrength) (LocalVariables|GetAttemptedHitIndex) (LocalVariables|GetTimeForNext)))))

(fn SetupDynamicMaterials
  (params (Mesh :type Static%20Mesh%20Component%20Object%20Reference) (MatArray :type Array%20of%20Material%20Instance%20Dynamic%20Object%20References))
  (for _index (range (Utilities|Array|Length (Rendering|Material|GetMaterials Mesh)))
    (bind _returnvalue (Rendering|Material|CreateDynamicMaterialInstance Mesh _index 0 "None" :!class /Script/Engine.PrimitiveComponent))
    (Utilities|Array|Add MatArray _returnvalue)))

(fn SetupRiverTint
  (params (EnableTint :type Boolean) (TintColor :type Linear%20Color%20Structure))
  (if EnableTint
    (Rendering|Material|SetVectorParameterValue "/Game/Athena/Environments/Landscape/MPC/MPC_Landscape.MPC_Landscape" "RiverTintAdd" TintColor :!class /Script/Engine.KismetMaterialLibrary)
    (else
      (Rendering|Material|SetVectorParameterValue "/Game/Athena/Environments/Landscape/MPC/MPC_Landscape.MPC_Landscape" "RiverTintAdd" (Utilities|Struct|MakeLinearColor 0.0) :!class /Script/Engine.KismetMaterialLibrary))))

(fn UpdateCrackProgression
  (params (Mesh :type Mesh%20Component%20Object%20Reference) (Progression :type Float%20%28double-precision%29) (Threshold :type Float%20%28double-precision%29))
  (Rendering|SetVisibility Mesh (> Progression Threshold) true)
  (Rendering|Material|SetScalarParameterValueonMaterials Mesh "CrackSize" (Math|Float|Clamp(Float) (- Progression Threshold))))

(fn ConstructionScript
  (bind _crackopacity (Variables|Default|GetCrackOpacity))
  (bind _rainbow (Variables|Default|GetRainbow))
  (bind _centralcrack (Variables|Default|GetCentralCrack))
  (Utilities|FlowControl|Sequence
    (:then_0
      ;[1 1 1] Update sky cracks
      (if (Variables|Default|GetPreviewCracks)
        (Rendering|SetVisibility (Variables|Default|GetSmallCrack1) (> _crackopacity 0.1) false)
        (Rendering|Material|SetScalarParameterValueonMaterials (Variables|Default|GetCentralCrack) "Opacity" (Variables|Default|GetCrackOpacity))
        (if (<= _crackopacity 0.0)
          (Components|StaticMesh|SetStaticMesh _centralcrack "/Figment_S05/Athena/Effects/Meshes/Geode/SM_Geode_Sky_Crack_Healed.SM_Geode_Sky_Crack_Healed")
          (Rendering|Material|SetScalarParameterValueonMaterials _centralcrack "MinOpacity" 0.5)
          (Rendering|Material|SetScalarParameterValueonMaterials _centralcrack "Corruption" (Variables|Default|GetCorruption))
          (Effects|Components|ParticleSystem|SetTemplate (Variables|Default|GetP_Geode_Crack_Center_Loop) "/Figment_S05/Athena/Effects/Particles/Geode/P_Geode_Crack_Center_ParticlesOnly_Corrupt_Red.P_Geode_Crack_Center_ParticlesOnly_Corrupt_Red"))))
    (:then_1
      (Rendering|SetVisibility _rainbow (Variables|InEditorTesting|GetShowRainbow) false)
      (Transformation|SetWorldTransform _rainbow (Variables|Default|GetRainbowWorldTransform) false false))
    (:then_2
      (CallFunction|SetupRiverTint (Variables|InEditorTesting|GetEnableRiverTint) (Variables|Default|GetRiverTintStPatrick's)))))
